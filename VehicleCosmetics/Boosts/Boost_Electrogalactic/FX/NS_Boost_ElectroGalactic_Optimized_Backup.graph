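# Niagara effect for the Electrogalactic vehicle boost cosmetic, named as a backup of the optimized version.
# A persistent sphere glow, lightning core and crumb sprites, and two trailing ribbons, one pulsing, form the plume.
# The core merges two sprite layers via spawn groups; sprites sit at user-supplied boost and car offsets, with their socket sampling disabled.
# Painted-variant palettes and frontend-preview switches retune color, motion and brightness.
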NiagaraSystem /VehicleCosmetics/Boosts/Boost_Electrogalactic/FX/NS_Boost_ElectroGalactic_Optimized_Backup
dependencies {
 usedDataInterfaces [
  &NiagaraDataInterfaceCurve
  &NiagaraDataInterfaceArrayColor
  &NiagaraDataInterfaceSpriteRendererInfo
  &NiagaraDataInterfaceArrayFloat
  &NiagaraDataInterfaceColorCurve
  &NiagaraDataInterfaceSkeletalMesh
 ]
 usedDynamicInputs []
 usedModules [
  &self:SystemSpawnScript.SystemScriptSource.SystemScriptGraph.NiagaraNodeAssignment_0.SetVariables_#1
  &SystemState
 ]
}
emitters {
 NE_Lightning_Core {
  emitterUpdateScript [
   &EmitterState(
    Inactive Response=::NewEnumerator1
    Life Cycle Mode=::NewEnumerator1
    Loop Duration=1
    Recalculate Duration Each Loop=true
    UseLoopDelay=false
   )
   &VC_FilteredSocketCount() {enabled=false}
   &SpawnRate(
    SpawnRate=&MakeCustomFloatFromBool(Bool=$User.bIsFrontendPreview, True Float=9, False Float=15)
    Use Spawn Probability=false
   )
   &SpawnBurst_Instantaneous(
    Spawn Count=&MakeCustomIntFromBool(Bool=$User.bIsFrontendPreview, True INT=5, False INT=10)
    Use Spawn Probability=false
    UseLoopCountLimit=false
   )
   &SpawnRate(SpawnGroup=1, SpawnRate=6, Use Spawn Probability=false) {module=SpawnRate001}
  ]
  particleSpawnScript [
   &VC_PaintedVariantsOverride(
    Black=<rgba 0.0001 0.0001 0.0001 3>
    BurntSienna=<rgba 0.5 0.0786 0.015 0.8>
    Cobalt=<rgba 0 0.0167 0.2 1.5>
    Crimson=<rgba 0.1 0.004 0.0056 0.85>
    DefaultColor=&SelectLinearColorFromArray(
     Array Sampling Mode=::NewEnumerator1
     Direct Array Index=&RandomRangeInt(Minimum=0, Maximum=1)
     Interpolation=&RandomRangeFloat(Minimum=0, Maximum=1)
     Color Selection Array=DataInterface() {propertyValues {ColorData [<rgba 1 0 0.8667 0.5>, <rgba 0 0.4864 0.5 0.4>]}}
     Random Seed=?
    )
    ForestGreen=<rgba 0.0833 1 0 1>
    Grey=<rgba 0.002 0.002 0.002 1>
    Lime=<rgba 0.1 0.15 0 1>
    Orange=<rgba 0.5 0.1758 0 0.7>
    Pink=<rgba 0.35 0.1052 0.3137 1>
    Purple=&MakeCustomLinearColorFromBool(Bool=$User.bIsFrontendPreview, True Color=<rgba 0.125 0 0.75 1>, False Color=<rgba 0.2875 0 0.75 1>)
    R1 Black:LinearColor=<rgba 0.0001 0.0001 0.0001 2>
    R1 BurntSienna:LinearColor=<rgba 0.3 0.025 0 1>
    R1 Cobalt:LinearColor=<rgba 0 0.0833 1 1>
    R1 Crimson:LinearColor=<rgba 0.05 0 0.0012 1>
    R1 DefaultColor=&SelectLinearColorFromArray(
     Array Sampling Mode=::NewEnumerator1
     Direct Array Index=&RandomRangeInt(Minimum=0, Maximum=1)
     Color Selection Array=DataInterface() {propertyValues {ColorData [<rgba 0.1667 0 1 1>, <rgba 0 0.025 0.1 0.5>]}}
     Random Seed=?
    )
    R1 ForestGreen:LinearColor=<rgba 0.05 0.3 0 0.8>
    R1 Grey:LinearColor=<rgba 0.0015 0.0015 0.0015 1>
    R1 Lime:LinearColor=<rgba 0.0667 0.1 0 1>
    R1 Orange:LinearColor=<rgba 0.25 0.05 0 0.7>
    R1 Pink:LinearColor=<rgba 0.35 0.1052 0.3137 1>
    R1 Purple:LinearColor=<rgba 0.2514 0 0.5 0.9>
    R1 Saffron:LinearColor=<rgba 0.3 0.075 0 1.2>
    R1 SkyBlue:LinearColor=<rgba 0 0.2323 0.35 0.75>
    R1 TitaniumWhite:LinearColor=<rgba 0.5 0.5 0.5 0.5>
    Saffron=<rgba 1.1 0.1833 0 1>
    SkyBlue=<rgba 0 0.2323 0.35 0.75>
    TitaniumWhite=<rgba 0.5 0.5 0.5 0.5>
    Use R1 Painted Overrides=true
   )
   /SetVariables(Particles.VisTag:NiagaraInt32=&SpawnGroup()) {isSetParameters=true}
   &InitializeParticle(
    Color Mode=::NewEnumerator0
    Lifetime=&RandomRangeFloat(Minimum=0.4, Maximum=1)
    Position Offset=&AddVectorToPosition(
     Position=&AddVectorToPosition(Position=<-90 0 0>, Vector=$User.SingleBoostPos)
     Vector=&Multiply_VectorByFloat(Vector=$User.DTCarOffset, Float=1)
    )
    Position Offset Coordinate Space=::Local
    Sprite Rotation Mode=::NewEnumerator2
    Sprite Size=&MaskVector2dBySpawnGroup(
     Masked Vector2D=&RandomRangeVector2D(Minimum=<30 25>, Maximum=<60 50>)
     Passthrough Vector2D=&RandomRangeVector2D(Minimum=<30 25>, Maximum=<42 35>)
     Spawn Group Mask=0
    )
    Sprite Size Mode=::NewEnumerator4
    UsePositionOffset=true
    Write Material Random=true
   )
   &PartitionParticles(Number Of Partitions=$Emitter.PortCount) {enabled=false}
   &SkeletalMeshLocation(
    Bone / Socket Index=$Particles.PartitionParticles.Partition
    Bone / Socket Sampling Mode=::NewEnumerator5
    MaskSpawnGroup=false
    Mesh Sampling Type=::NewEnumerator6
    Sampled Position Offset=&MakeCustomVectorFromBool(
     Bool=$User.bIsFrontendPreview
     True Vector=&AddVectorToPosition(Position=<0 0 -12.5>, Vector=$User.DTCarOffset)
     False Vector=<0 0 0>
    )
    Sampled Position Offset Coordinate Space:ENiagaraSkelMeshTransforms=::NewEnumerator2
    Skeletal Mesh=$System.SkeletalMesh
   ) {enabled=false}
   &AddVelocity(
    Cone Angle=&MakeCustomFloatFromBool(Bool=$User.bIsFrontendPreview, True Float=5, False Float=0)
    Module.UseConeFalloff=false
    Module.UseVelDistribution=false
    Override Seed=false
    Velocity Mode=::NewEnumerator2
    Velocity Speed=/VehicleCosmetics/FX/Niagara/Modules/MergeEmitters/VC_FloatBySpawnGroup(
     SpawnGroup or ExecIndex:NiagaraInt32=0
     CustomInt:NiagaraInt32=0
     MaximumSpawnGroups:NiagaraInt32=2
     Float SG 0:NiagaraFloat=0
     Float SG 1:NiagaraFloat=-1000
     Float SG 2:NiagaraFloat=0
     Float SG 3:NiagaraFloat=0
     Float SG 4:NiagaraFloat=0
     Float SG 5:NiagaraFloat=0
    )
   )
   &ShapeLocation(
    MaskSpawnGroup=false
    Override Seed=false
    Sphere Radius=&RandomRangeFloat(Minimum=20, Maximum=10)
    UseOwnerScale=true
   )
  ]
  particleUpdateScript [
   &ParticleState()
   &SpriteRotationRate(Rotation Rate=&RandomRangeFloat(
     Minimum=&VC_FloatBySpawnGroup(
      SpawnGroup or ExecIndex:NiagaraInt32=0
      CustomInt:NiagaraInt32=0
      MaximumSpawnGroups:NiagaraInt32=2
      Float SG 0:NiagaraFloat=&RandomRangeFloat(Minimum=-25, Maximum=-100)
      Float SG 1:NiagaraFloat=-200
      Float SG 2:NiagaraFloat=0
      Float SG 3:NiagaraFloat=0
      Float SG 4:NiagaraFloat=0
      Float SG 5:NiagaraFloat=0
     )
     Maximum=&VC_FloatBySpawnGroup(
      SpawnGroup or ExecIndex:NiagaraInt32=0
      CustomInt:NiagaraInt32=0
      MaximumSpawnGroups:NiagaraInt32=2
      Float SG 0:NiagaraFloat=&RandomRangeFloat(Minimum=25, Maximum=100)
      Float SG 1:NiagaraFloat=200
      Float SG 2:NiagaraFloat=0
      Float SG 3:NiagaraFloat=0
      Float SG 4:NiagaraFloat=0
      Float SG 5:NiagaraFloat=0
     )
    ))
   &ScaleSpriteSize(
    Scale Sprite Size Mode=::NewEnumerator0
    Uniform Scale Factor=&VC_FloatBySpawnGroup(
     SpawnGroup or ExecIndex:NiagaraInt32=0
     CustomInt:NiagaraInt32=0
     MaximumSpawnGroups:NiagaraInt32=2
     Float SG 0:NiagaraFloat=&RandomRangeFloat(
      Minimum=&FloatFromCurve(
       FloatCurve=DataInterface() {propertyValues {Curve {keys [
           {interpMode=RCIM_Cubic, tangentMode=RCTM_User, time=0, value=0.2}
           {interpMode=RCIM_Cubic, tangentMode=RCTM_User, time=0.5, value=1}
           {tangentMode=RCTM_User, time=1, value=0.2}
          ]}}}
       Scale Curve=2.8
      )
      Maximum=&FloatFromCurve(
       FloatCurve=DataInterface() {propertyValues {Curve {keys [{tangentMode=RCTM_User, time=0, value=0.5}, <at 0.5 1>, {tangentMode=RCTM_User, time=1, value=0.5}]}}}
       Scale Curve=2
      )
     )
     Float SG 1:NiagaraFloat=&RandomRangeFloat(
      Minimum=&FloatFromCurve(
       FloatCurve=DataInterface() {propertyValues {Curve {keys [<at 0 1>,<at 1 0.25>]}}}
       Scale Curve=&MakeCustomFloatFromBool(Bool=$User.bIsFrontendPreview, True Float=2.5, False Float=3)
      )
      Maximum=&FloatFromCurve(FloatCurve=DataInterface() {propertyValues {Curve {keys [<at 0 1>,<at 1 1>]}}}, Scale Curve=2)
     )
     Float SG 2:NiagaraFloat=0
     Float SG 3:NiagaraFloat=0
     Float SG 4:NiagaraFloat=0
     Float SG 5:NiagaraFloat=0
    )
   )
   &Drag(
    Drag=&VC_FloatBySpawnGroup(
     SpawnGroup or ExecIndex:NiagaraInt32=0
     CustomInt:NiagaraInt32=0
     MaximumSpawnGroups:NiagaraInt32=2
     Float SG 0:NiagaraFloat=&FloatFromCurve(FloatCurve=DataInterface() {propertyValues {Curve {keys [{tangentMode=RCTM_User, time=0.1, value=0}, {tangentMode=RCTM_User, time=0.5, value=1}]}}})
     Float SG 1:NiagaraFloat=0
     Float SG 2:NiagaraFloat=0
     Float SG 3:NiagaraFloat=0
     Float SG 4:NiagaraFloat=0
     Float SG 5:NiagaraFloat=0
    )
    Use Linear Drag=true
    Use Rotational Drag=true
   )
   &SolveForcesAndVelocity(Clamp Velocity=false, Limit Acceleration=false)
   &DynamicMaterialParameters(
    Index 0 Param 2=&FloatFromCurve(
     FloatCurve=DataInterface() {propertyValues {Curve {keys [{interpMode=RCIM_Cubic, tangentMode=RCTM_User, time=0, value=1}, {interpMode=RCIM_Cubic, tangentMode=RCTM_User, time=1, value=0}]}}}
     Scale Curve=&MakeCustomFloatFromBool(
      Bool=$User.bIsPainted
      True Float=&MakeCustomFloatFromBool(Bool=$User.bIsFrontendPreview, True Float=5, False Float=10)
      False Float=5
     )
    )
    Index 0 Param 3=1
    Index 0 Param 4=&RandomRangeFloat(Minimum=0.33, Maximum=0.8)
    Index 1 Param 1=&MakeCustomFloatFromBool(
     Bool=$User.bIsPainted
     True Float=&MakeCustomFloatFromBool(Bool=$User.bIsFrontendPreview, True Float=0.4, False Float=0.1)
     False Float=0.05
    )
    Index 2 Param 1=10
    Index 2 Param 2=&FloatFromCurve(
     FloatCurve=DataInterface() {propertyValues {Curve {keys [{interpMode=RCIM_Cubic, tangentMode=RCTM_User, time=0, value=1}, {interpMode=RCIM_Cubic, tangentMode=RCTM_User, time=1, value=0}]}}}
     Scale Curve=15
    )
    Index 2 Param 3=1
    Index 2 Param 4=&RandomRangeFloat(Minimum=0.1, Maximum=0.5, Evaluation Type=::NewEnumerator1)
    Index 3 Param 1=&MakeCustomFloatFromBool(Bool=$User.bIsPainted, True Float=0.4, False Float=0.05)
    Module.Index0Param0Write=true
    Module.Index0Param1Write=true
    Module.Index0Param2Write=true
    Module.Index0Param3Write=true
    Module.Index1Param0Write=true
    Module.Index1Param1Write=false
    Module.Index1Param2Write=false
    Module.Index1Param3Write=false
    Module.Index2Param0Write=true
    Module.Index2Param1Write=true
    Module.Index2Param2Write=true
    Module.Index2Param3Write=true
    Module.Index3Param0Write=true
    Module.Index3Param1Write=true
    Module.Index3Param2Write=true
    Module.Index3Param3Write=true
    Write Parameter Index 1=true
    Write Parameter Index 2=true
    Write Parameter Index 3=true
   )
   &SubUVAnimation(
    Random Start Frame=true
    "Sprite Renderer "=DataInterface() {propertyValues {SpriteRenderer=&self:NE_Lightning_Core_0.NiagaraSpriteRendererProperties_0}}
    Start Frame Offset:NiagaraInt32=&RandomRangeInt(Minimum=0, Maximum=1)
    SubUV Animation Mode:ENiagara_SubUVLookupModeV2=::NewEnumerator1
    SubUV Curve=DataInterface() {propertyValues {Curve {keys [
        {interpMode=RCIM_Cubic, tangentMode=RCTM_User, time=0, value=0}
        {arriveTangent=0.7938, interpMode=RCIM_Cubic, leaveTangent=0.7938, tangentMode=RCTM_User, time=1, value=0.2}
       ]}}}
    SubUV Curve Index:NiagaraFloat=$Particles.Lifetime
    UseEndFrame=false
    UseStartFrame=false
   ) {module=SubUVAnimation001}
   &SubUVAnimation(
    Output Alternate Renderer Binding=true
    "Sprite Renderer "=DataInterface() {propertyValues {SpriteRenderer=&self:NE_Lightning_Core_6.NiagaraSpriteRendererProperties_0}}
    SubUV Animation Mode:ENiagara_SubUVLookupModeV2=::NewEnumerator1
    SubUV Curve=DataInterface() {propertyValues {Curve {keys [{tangentMode=RCTM_User, time=0, value=0}, {tangentMode=RCTM_User, time=1, value=2}]}}}
    UseEndFrame=false
    UseStartFrame=false
   )
   &ScaleColor(
    Scale Alpha=&VC_FloatBySpawnGroup(
     SpawnGroup or ExecIndex:NiagaraInt32=0
     CustomInt:NiagaraInt32=0
     MaximumSpawnGroups:NiagaraInt32=2
     Float SG 0:NiagaraFloat=&FloatFromCurve(FloatCurve=DataInterface() {propertyValues {Curve {keys [
          {interpMode=RCIM_Cubic, tangentMode=RCTM_User, time=0, value=0.94}
          {interpMode=RCIM_Cubic, tangentMode=RCTM_User, time=0.8, value=0.94}
          {arriveTangent=-5.9258, interpMode=RCIM_Cubic, leaveTangent=-5.9258, tangentMode=RCTM_User, time=1, value=0}
         ]}}})
     Float SG 1:NiagaraFloat=&MakeCustomFloatFromBool(
      Bool=$User.bIsFrontendPreview
      True Float=&MakeCustomFloatFromBool(
       Bool=$User.bIsPainted
       True Float=&FloatFromCurve(
        FloatCurve=DataInterface() {propertyValues {Curve {keys [
            {interpMode=RCIM_Cubic, tangentMode=RCTM_User, time=0, value=0.94}
            {interpMode=RCIM_Cubic, tangentMode=RCTM_User, time=0.5, value=0.94}
            {arriveTangent=-5.9258, interpMode=RCIM_Cubic, leaveTangent=-5.9258, tangentMode=RCTM_User, time=0.8, value=0}
           ]}}}
        Scale Curve=2
       )
       False Float=&FloatFromCurve(
        FloatCurve=DataInterface() {propertyValues {Curve {keys [
            {interpMode=RCIM_Cubic, tangentMode=RCTM_User, time=0, value=0.94}
            {arriveTangent=-0.2129, interpMode=RCIM_Cubic, leaveTangent=-0.2129, tangentMode=RCTM_User, time=1, value=0}
           ]}}}
        Scale Curve=2
       )
      )
      False Float=&MakeCustomFloatFromBool(
       Bool=$User.bIsPainted
       True Float=&FloatFromCurve(
        FloatCurve=DataInterface() {propertyValues {Curve {keys [
            {interpMode=RCIM_Cubic, tangentMode=RCTM_User, time=0, value=0.94}
            {interpMode=RCIM_Cubic, tangentMode=RCTM_User, time=0.3, value=0.95}
            {arriveTangent=-5.9258, interpMode=RCIM_Cubic, leaveTangent=-5.9258, tangentMode=RCTM_User, time=0.6, value=0}
           ]}}}
        Scale Curve=2
       )
       False Float=&FloatFromCurve(
        FloatCurve=DataInterface() {propertyValues {Curve {keys [
            {interpMode=RCIM_Cubic, tangentMode=RCTM_User, time=0, value=0.94}
            {arriveTangent=-0.2129, interpMode=RCIM_Cubic, leaveTangent=-0.2129, tangentMode=RCTM_User, time=0.7, value=0}
           ]}}}
        Scale Curve=2
       )
      )
     )
     Float SG 2:NiagaraFloat=0
     Float SG 3:NiagaraFloat=0
     Float SG 4:NiagaraFloat=0
     Float SG 5:NiagaraFloat=0
    )
    Scale RGB=/VehicleCosmetics/FX/Niagara/Modules/MergeEmitters/VC_VectorBySpawnGroup(
     SpawnGroup or ExecIndex:NiagaraInt32=0
     CustomInt:NiagaraInt32=0
     MaximumSpawnGroups:NiagaraInt32=2
     Vector Group 0:Vector3f=<0.5 0.5 0.5>
     Vector Group 1:Vector3f=<1 1 1>
     Vector Group 2:Vector3f=<0 0 0>
     Vector Group 3:Vector3f=<0 0 0>
     Vector Group 4:Vector3f=<0 0 0>
     Vector Group 5:Vector3f=<0 0 0>
    )
    ScaleA=true
    ScaleRGB=true
   ) {module=ScaleColor001}
  ]
  rendererClasses [&NiagaraSpriteRendererProperties]
  renderers [
   {rendererClass=&NiagaraSpriteRendererProperties}
   {rendererClass=&NiagaraSpriteRendererProperties, rendererIndex=1}
  ]
 }
 NE_Lightning_crumbs {
  emitterSpawnScript [/SetVariables(Emitter.Is NOT Frontend:NiagaraBool=&Boolean_NotOperation(Bool=$User.bIsFrontendPreview)) {isSetParameters=true}]
  emitterUpdateScript [
   &EmitterState(Life Cycle Mode=::NewEnumerator1, Loop Duration=1, UseLoopDelay=false)
   &VC_FilteredSocketCount() {enabled=false}
   &SpawnRate(
    SpawnRate=&MakeCustomFloatFromBool(Bool=$User.bIsFrontendPreview, True Float=20, False Float=20)
    Use Spawn Probability=false
   )
  ]
  particleSpawnScript [
   &VC_PaintedVariantsOverride(
    Black=<rgba 0.0001 0.0001 0.0001 2>
    BurntSienna=<rgba 0.3 0.025 0 1>
    Cobalt=<rgba 0 0.0833 1 1>
    Crimson=<rgba 0.05 0 0.0012 1>
    DefaultColor=&SelectLinearColorFromArray(
     Array Sampling Mode=::NewEnumerator1
     Direct Array Index=&RandomRangeInt(Minimum=0, Maximum=1)
     Color Selection Array=DataInterface() {propertyValues {ColorData [<rgba 0.6667 0 1 1>, <rgba 0 0.1 0.4 0.5>]}}
     Random Seed=?
    )
    ForestGreen=<rgba 0.05 0.3 0 0.8>
    Grey=<rgba 0.0015 0.0015 0.0015 1>
    Lime=<rgba 0.0667 0.1 0 1>
    Orange=<rgba 0.25 0.05 0 0.7>
    Pink=<rgba 0.35 0.1052 0.3137 1>
    Purple=<rgba 0.2514 0 0.5 0.9>
    Saffron=<rgba 0.3 0.075 0 1.2>
    SkyBlue=<rgba 0 0.2323 0.35 0.75>
    TitaniumWhite=<rgba 0.5 0.5 0.5 0.5>
   )
   &InitializeParticle(
    Color Mode=::NewEnumerator0
    Position Offset=&AddVectorToPosition(
     Position=&AddVectorToPosition(Position=<-110 0 0>, Vector=$User.SingleBoostPos)
     Vector=&Multiply_VectorByFloat(Vector=$User.DTCarOffset, Float=1)
    )
    Position Offset Coordinate Space=::Local
    Sprite Rotation Mode=::NewEnumerator2
    Sprite Size=&Multiply_Vector2D(A=<60 50>, B=&RandomRangeVector2D(Minimum=<0.5 0.5>, Maximum=<0.7 0.7>))
    Sprite Size Mode=::NewEnumerator4
    UsePositionOffset=true
    Write Material Random=true
   )
   &PartitionParticles(Number Of Partitions=$Emitter.PortCount) {enabled=false}
   &SkeletalMeshLocation(
    Bone / Socket Index=$Particles.PartitionParticles.Partition
    Bone / Socket Sampling Mode=::NewEnumerator5
    MaskSpawnGroup=false
    Mesh Sampling Type=::NewEnumerator6
    Sampled Position Offset=&MakeCustomVectorFromBool(
     Bool=$User.bIsFrontendPreview
     True Vector=&AddVectorToPosition(Position=<0 0 -12.5>, Vector=$User.DTCarOffset)
     False Vector=<0 0 0>
    )
    Sampled Position Offset Coordinate Space:ENiagaraSkelMeshTransforms=::NewEnumerator2
    Skeletal Mesh=$System.SkeletalMesh
   ) {enabled=false}
   &AddVelocity(Velocity=<1000 0 0>, Velocity Speed Scale=-1)
   &AddVelocity(
    Cone Angle=&MakeCustomFloatFromBool(Bool=$User.bIsFrontendPreview, True Float=5, False Float=0)
    Cone Axis=<-1 0 0>
    Module.UseConeFalloff=false
    Module.UseVelDistribution=false
    Override Seed=false
    Velocity Mode=::NewEnumerator2
    Velocity Speed=1000
   ) {enabled=false, module=AddVelocity001}
  ]
  particleUpdateScript [
   &ParticleState()
   &SpriteRotationRate(Rotation Rate=&RandomRangeFloat(Minimum=-200, Maximum=200))
   &InheritSourceMovement(
    Applied Position Scale=&Multiply_VectorByFloat(
     Vector=<1 1 1>
     Float=&MakeCustomFloatFromBool(
      Bool=$User.bIsFrontendPreview
      True Float=&FloatFromCurve(FloatCurve=DataInterface() {propertyValues {Curve {keys [
           {interpMode=RCIM_Cubic, leaveTangentWeight=0.247, tangentMode=RCTM_User, time=0.3, value=1}
           {arriveTangent=-0.6637, interpMode=RCIM_Cubic, leaveTangent=-0.6637, tangentMode=RCTM_User, time=1, value=0.8}
          ]}}})
      False Float=&FloatFromCurve(FloatCurve=DataInterface() {propertyValues {Curve {keys [
           {arriveTangent=-0.2262, interpMode=RCIM_Cubic, leaveTangent=-0.2262, leaveTangentWeight=0.247, tangentMode=RCTM_User, time=0, value=1}
           {arriveTangentWeight=0.2337, interpMode=RCIM_Cubic, tangentMode=RCTM_User, time=1, value=0.5}
          ]}}})
     )
    )
    ApplyForce=false
    ApplyPosition=true
    ApplyPositionFalloff=false
   )
   &ScaleSpriteSize(
    Scale Sprite Size Mode=::NewEnumerator0
    Uniform Scale Factor=&RandomRangeFloat(
     Minimum=&FloatFromCurve(
      FloatCurve=DataInterface() {propertyValues {Curve {keys [<at 0 1>,<at 1 0.25>]}}}
      Scale Curve=&MakeCustomFloatFromBool(Bool=$User.bIsFrontendPreview, True Float=2.5, False Float=3)
     )
     Maximum=&FloatFromCurve(FloatCurve=DataInterface() {propertyValues {Curve {keys [<at 0 1>,<at 1 1>]}}}, Scale Curve=2)
    )
   )
   &Drag(
    Drag=&FloatFromCurve(
     FloatCurve=DataInterface() {propertyValues {Curve {keys [{tangentMode=RCTM_User, time=0.1, value=0}, {tangentMode=RCTM_User, time=0.5, value=1}]}}}
     Scale Curve=&MakeCustomFloatFromBool(Bool=$User.bIsFrontendPreview, True Float=1, False Float=0)
    )
    Use Linear Drag=true
    Use Rotational Drag=true
   )
   &SolveForcesAndVelocity(Clamp Velocity=false, Limit Acceleration=false)
   &DynamicMaterialParameters(
    Index 0 Param 1=10
    Index 0 Param 2=&FloatFromCurve(
     FloatCurve=DataInterface() {propertyValues {Curve {keys [{interpMode=RCIM_Cubic, tangentMode=RCTM_User, time=0, value=1}, {interpMode=RCIM_Cubic, tangentMode=RCTM_User, time=1, value=0}]}}}
     Scale Curve=15
    )
    Index 0 Param 3=1
    Index 0 Param 4=&RandomRangeFloat(Minimum=0.1, Maximum=0.5, Evaluation Type=::NewEnumerator1)
    Index 1 Param 1=&MakeCustomFloatFromBool(Bool=$User.bIsPainted, True Float=0.4, False Float=0.05)
    Module.Index0Param0Write=true
    Module.Index0Param1Write=true
    Module.Index0Param2Write=true
    Module.Index0Param3Write=true
    Module.Index1Param0Write=true
    Module.Index1Param1Write=false
    Module.Index1Param2Write=false
    Module.Index1Param3Write=false
    Write Parameter Index 1=true
   )
   &SubUVAnimation(
    "Sprite Renderer "=DataInterface() {propertyValues {SpriteRenderer=&self:NE_Lightning_Core_6.NiagaraSpriteRendererProperties_0}}
    SubUV Animation Mode:ENiagara_SubUVLookupModeV2=::NewEnumerator1
    SubUV Curve=DataInterface() {propertyValues {Curve {keys [{tangentMode=RCTM_User, time=0, value=0}, {tangentMode=RCTM_User, time=1, value=2}]}}}
    UseEndFrame=false
    UseStartFrame=false
   )
   &ScaleColor(
    Scale Alpha=&MakeCustomFloatFromBool(
     Bool=$User.bIsFrontendPreview
     True Float=&MakeCustomFloatFromBool(
      Bool=$User.bIsPainted
      True Float=&FloatFromCurve(
       FloatCurve=DataInterface() {propertyValues {Curve {keys [
           {interpMode=RCIM_Cubic, tangentMode=RCTM_User, time=0, value=0.94}
           {interpMode=RCIM_Cubic, tangentMode=RCTM_User, time=0.5, value=0.94}
           {arriveTangent=-5.9258, interpMode=RCIM_Cubic, leaveTangent=-5.9258, tangentMode=RCTM_User, time=1, value=0}
          ]}}}
       Scale Curve=2
      )
      False Float=&FloatFromCurve(
       FloatCurve=DataInterface() {propertyValues {Curve {keys [
           {interpMode=RCIM_Cubic, tangentMode=RCTM_User, time=0, value=0.94}
           {arriveTangent=-0.2129, interpMode=RCIM_Cubic, leaveTangent=-0.2129, tangentMode=RCTM_User, time=1, value=0}
          ]}}}
       Scale Curve=2
      )
     )
     False Float=&MakeCustomFloatFromBool(
      Bool=$User.bIsPainted
      True Float=&FloatFromCurve(FloatCurve=DataInterface() {propertyValues {Curve {keys [
           {interpMode=RCIM_Cubic, tangentMode=RCTM_User, time=0, value=0.94}
           {interpMode=RCIM_Cubic, tangentMode=RCTM_User, time=0.3, value=0.95}
           {arriveTangent=-5.9258, interpMode=RCIM_Cubic, leaveTangent=-5.9258, tangentMode=RCTM_User, time=0.6, value=0}
          ]}}})
      False Float=&FloatFromCurve(FloatCurve=DataInterface() {propertyValues {Curve {keys [
           {interpMode=RCIM_Cubic, tangentMode=RCTM_User, time=0, value=0.94}
           {arriveTangent=-0.2129, interpMode=RCIM_Cubic, leaveTangent=-0.2129, tangentMode=RCTM_User, time=0.7, value=0}
          ]}}})
     )
    )
    Scale RGB=&Multiply_VectorByFloat(Vector=<1 1 1>, Float=0.5)
    ScaleA=true
    ScaleRGB=true
   )
  ]
  rendererClasses [&NiagaraSpriteRendererProperties]
  renderers [{rendererClass=&NiagaraSpriteRendererProperties}]
 }
 NE_VC_Boost_MAIN_RIBBON001 {
  emitterUpdateScript [
   &EmitterState(
    Inactive Response=::NewEnumerator1
    Life Cycle Mode=::NewEnumerator1
    Loop Duration=0.1
    Recalculate Duration Each Loop=true
    UseLoopDelay=false
   )
   &VC_FilteredSocketCount()
   &VC_RateQualityHandler(Epic Spawn Rate=30, High Spawn Rate=30, Low Spawn Rate=30, Use Custom Rate?=true) {module=VC_RateQualityHandler001}
   &VC_RateQualityHandler()
   &VC_SpawnRibbon_v1(Spawn Time=-1, SpawnRate=60)
  ]
  particleSpawnScript [
   &VC_PaintedVariantsOverride(
    Black=<rgba 0.002 0.002 0.002 0.8>
    BurntSienna=<rgba 0.03 0.0134 0.0015 0.25>
    Cobalt=<rgba 0.06 0.1067 0.2 0.01>
    Crimson=<rgba 0.7 0 0.0117 0.001>
    DefaultColor=&MakeCustomLinearColorFromBool(Bool=$User.bIsFrontendPreview, True Color=<rgba 0.0187 0 0.08 1>, False Color=<rgba 0.0408 0 0.35 1>)
    ForestGreen=<rgba 0.02 0.03 0 0.22>
    Grey=<rgba 0.01 0.01 0.01 0.25>
    Lime=<rgba 0.3333 0.4 0 0.003>
    Orange=<rgba 0.35 0.084 0.0175 0.015>
    Pink=<rgba 1 0.4 0.75 0.001>
    Purple=<rgba 0.016 0.0075 0.05 0.7>
    R1 Black:LinearColor=<rgba 0.002 0.002 0.002 0.8>
    R1 BurntSienna:LinearColor=<rgba 2.5 0.9231 0.5 40>
    R1 Cobalt:LinearColor=<rgba 0 0.2775 15 50>
    R1 Crimson:LinearColor=<rgba 1.8 0.63 0.63 100>
    R1 DefaultColor=&MakeCustomLinearColorFromBool(Bool=$User.bIsFrontendPreview, True Color=<rgba 166.6667 0 500 1>, False Color=<rgba 0.0408 0 0.35 1>)
    R1 ForestGreen:LinearColor=<rgba 14.3333 20 3 1>
    R1 Grey:LinearColor=<rgba 0.8 0.8 0.8 60>
    R1 Lime:LinearColor=<rgba 0.8917 1 0.35 350>
    R1 Orange:LinearColor=<rgba 10 3.488 0 20>
    R1 Pink:LinearColor=<rgba 20 2.3958 13.5985 25>
    R1 Purple:LinearColor=<rgba 4.1041 0 10 15>
    R1 Saffron:LinearColor=<rgba 100 76.6667 0 1>
    R1 SkyBlue:LinearColor=<rgba 0 23.2885 35 2>
    R1 TitaniumWhite:LinearColor=<rgba 100 100 100 1>
    Saffron=<rgba 0.04 0.032 0 0.3>
    SkyBlue=<rgba 0.02 0.095 0.2 0.03>
    TitaniumWhite=<rgba 0.2 0.2 0.2 0.005>
    Use R1 Painted Overrides=true
   )
   &VC_InitializeRibbon_v1(
    Color=$Particles.Color
    Custom Offset=true
    Lifetime=0.6
    Local Vector Offset=&AddVectorToPosition(
     Position=&AddVectorToPosition(
      Position=/Niagara/DynamicInputs/TypeConversions/MakeCustomPositionFromBool(Bool:NiagaraBool=$User.bIsFrontendPreview, True Position:NiagaraPosition=<-15 -5 0>, False Position:NiagaraPosition=<-25 -5 0>)
      Vector=$User.SingleBoostPos
     )
     Vector=&Multiply_VectorByFloat(Vector=$User.DTCarOffset, Float=1)
    )
    Offset Amount=1
    Ribbon Width=150
   )
   &AddVelocity(
    Velocity=<3000 0 0>
    Velocity Speed Scale=&MakeCustomFloatFromBool(Bool=$User.bIsFrontendPreview, True Float=-1, False Float=-0.8)
   ) {module=AddVelocity001}
   &ShapeLocation(MaskSpawnGroup=false, Override Seed=false, Sphere Radius=5, UseOwnerScale=true) {enabled=false}
  ]
  particleUpdateScript [
   &ParticleState()
   &InheritSourceMovement(
    Applied Position Scale=&Multiply_VectorByFloat(
     Vector=<1 1 1>
     Float=&FloatFromCurve(FloatCurve=DataInterface() {propertyValues {Curve {keys [<at 0 1 RCIM_Cubic>,<at 1 0 RCIM_Cubic>]}}})
    )
    ApplyForce=false
    ApplyPosition=true
    ApplyPositionFalloff=false
   )
   &SolveForcesAndVelocity(Clamp Velocity=false, Limit Acceleration=false)
   &VC_UpdateRibbon_v1()
   /SetVariables(Particles.Linear Color=&MultiplyLinearColor(Linear Color=$StackContext.Color, Scale Factor=<100 100 100 1>)) {isSetParameters=true}
   &ScaleColor(
    Scale Alpha=&MakeCustomFloatFromBool(
     Bool=$User.bIsPainted
     True Float=&FloatFromCurve(FloatCurve=DataInterface() {propertyValues {Curve {keys [
          {arriveTangent=-2.3938, interpMode=RCIM_Cubic, leaveTangent=-2.3938, tangentMode=RCTM_User, time=0.1, value=1}
          {arriveTangent=0.0152, interpMode=RCIM_Cubic, leaveTangent=0.0152, tangentMode=RCTM_User, time=1, value=1}
         ]}}})
     False Float=&FloatFromCurve(FloatCurve=DataInterface() {propertyValues {Curve {keys [{interpMode=RCIM_Cubic, tangentMode=RCTM_User, time=0.1, value=1}, {interpMode=RCIM_Cubic, tangentMode=RCTM_User, time=0.9, value=0}]}}})
    )
    Scale RGB=&Multiply_VectorByFloat(
     Vector=<1 1 1>
     Float=&FloatFromCurve(
      FloatCurve=DataInterface() {propertyValues {Curve {keys [<at 0 1 RCIM_Cubic>, {arriveTangent=-0.0159, interpMode=RCIM_Cubic, leaveTangent=-0.0159, tangentMode=RCTM_User, time=1, value=1}]}}}
      Scale Curve=&MakeCustomFloatFromBool(Bool=$User.bIsFrontendPreview, True Float=4.5, False Float=4)
     )
    )
    ScaleA=true
    ScaleRGB=true
   )
   &DynamicMaterialParameters(
    Index 0 Param 2=&FloatFromCurve(
     FloatCurve=DataInterface() {propertyValues {Curve {keys [<at 0 2>,<at 1 0>]}}}
     Scale Curve=&MakeCustomFloatFromBool(Bool=$User.bIsFrontendPreview, True Float=$User.Scale Curve_Index 0 Param 2, False Float=0.1)
    )
    Module.Index0Param0Write=true
    Module.Index0Param1Write=true
    Module.Index0Param2Write=true
    Module.Index0Param3Write=true
   )
   &ScaleRibbonWidth(Ribbon Width Scale=&FloatFromCurve(
     FloatCurve=DataInterface() {propertyValues {Curve {keys [
         {arriveTangent=1.6817, interpMode=RCIM_Cubic, leaveTangent=1.6817, tangentMode=RCTM_User, time=0, value=0.1}
         {arriveTangent=0.5, interpMode=RCIM_Cubic, leaveTangent=0.5, time=0.4, value=0.75}
         {arriveTangent=0.094, interpMode=RCIM_Cubic, leaveTangent=0.094, tangentMode=RCTM_User, time=0.8, value=0.5}
        ]}}}
     Scale Curve=1.15
    )) {module=ScaleRibbonWidth001}
  ]
  rendererClasses [&NiagaraRibbonRendererProperties]
  renderers [
   {rendererClass=&NiagaraRibbonRendererProperties}
   {rendererClass=&NiagaraRibbonRendererProperties, rendererIndex=1}
  ]
 }
 NE_VC_Boost_SEC_RIBBON {
  emitterSpawnScript [/SetVariables(Emitter.ContrastBackPerColor:NiagaraFloat=&SelectFloatFromArray(
     Direct Array Index=$User.ColorVariantID
     Float Selection Array=DataInterface() {propertyValues {FloatData [1,1,1,1,4,4.5,4,4,4,8,4,6,4,4,6]}}
    )) {isSetParameters=true}]
  emitterUpdateScript [
   &EmitterState(Life Cycle Mode=::NewEnumerator1, Loop Duration=1, UseLoopDelay=false)
   &VC_FilteredSocketCount()
   &VC_RateQualityHandler(Epic Spawn Rate=50, High Spawn Rate=50, Low Spawn Rate=40, Medium Spawn Rate=40, Use Custom Rate?=true)
   &VC_SpawnRibbon_v1(Spawn Time=-1, SpawnRate=$Emitter.Spawn Rate)
  ]
  particleSpawnScript [
   &VC_PaintedVariantsOverride(
    Black=<rgba 0.35 0.35 0.35 0.513>
    BurntSienna=<rgba 2.5 1.3917 0.75 0.4261>
    Cobalt=<rgba 1.5 1.925 10 0.2>
    Crimson=<rgba 8 0.056 0.1884 0.4>
    ForestGreen=<rgba 3.2667 7 1.4 0.435>
    Grey=<rgba 2.5 2.5 2.5 0.5>
    Lime=<rgba 4.1667 5 0 0.4>
    Orange=<rgba 8 4.7333 0.16 0.25>
    Pink=<rgba 6 3.9 5.125 0.2>
    Purple=<rgba 3.6667 0.5 10 0.4>
    Saffron=<rgba 5 4.4167 0 0.35>
    SkyBlue=<rgba 1.925 4.3083 5.5 0.2>
    TitaniumWhite=<rgba 7 7 7 0.15>
   )
   &VC_InitializeRibbon_v1(
    Color=$Particles.Color
    Custom Offset=true
    Lifetime=&MakeCustomFloatFromBool(Bool=$User.bIsFrontendPreview, True Float=1.4, False Float=0.7)
    Local Vector Offset=&AddVectorToPosition(
     Position=&AddVectorToPosition(
      Position=&MakeCustomVectorFromBool(Bool=$User.bIsFrontendPreview, True Vector=<-60 0 0>, False Vector=<-15 0 0>)
      Vector=$User.SingleBoostPos
     )
     Vector=$User.DTCarOffset
    )
    Offset Amount=1
    Ribbon Width=130
   )
   &AddVelocity(Velocity=<1800 0 0>, Velocity Speed Scale=-1) {module=AddVelocity001}
  ]
  particleUpdateScript [
   &ParticleState()
   &InheritSourceMovement(
    Applied Position Scale=&Multiply_VectorByFloat(
     Vector=<1 1 1>
     Float=&MakeCustomFloatFromBool(
      Bool=$User.bIsFrontendPreview
      True Float=&FloatFromCurve(FloatCurve=DataInterface() {propertyValues {Curve {keys [<at 0 1 RCIM_Cubic>,<at 1 0 RCIM_Cubic>]}}})
      False Float=&FloatFromCurve(FloatCurve=DataInterface() {propertyValues {Curve {keys [{arriveTangent=-0.0696, interpMode=RCIM_Cubic, leaveTangent=-0.0696, tangentMode=RCTM_User, time=0, value=1}, <at 0.6 0 RCIM_Cubic>]}}})
     )
    )
    ApplyForce=false
    ApplyPosition=true
    ApplyPositionFalloff=false
   )
   &SolveForcesAndVelocity(Clamp Velocity=false, Limit Acceleration=false)
   &VC_UpdateRibbon_v1()
   &ScaleRibbonWidth(Ribbon Width Scale=&FloatFromCurve(
     FloatCurve=DataInterface() {propertyValues {Curve {keys [
         {interpMode=RCIM_Cubic, tangentMode=RCTM_User, time=0, value=1.5}
         {arriveTangent=-0.7, interpMode=RCIM_Cubic, leaveTangent=-0.7, time=0.05, value=1.1}
         {interpMode=RCIM_Cubic, tangentMode=RCTM_User, time=1, value=0.8}
        ]}}}
     Scale Curve=0.8
    ))
   &ScaleColor(
    Scale Alpha=&Multiply_Float(
     A=&MakeCustomFloatFromBool(Bool=$User.bIsPainted, True Float=2, False Float=1)
     B=&Lerp_Float(
      A=&FloatFromCurve(FloatCurve=DataInterface() {propertyValues {Curve {keys [<at 0 5>,<at 0.6 0>]}}})
      B=&FloatFromCurve(FloatCurve=DataInterface() {propertyValues {Curve {keys [<at 0 1>,<at 0.6 0>]}}})
      Alpha=HlslExpression() {hlslExpression="lerp(0, .8, (sin(Engine.Time * 3.14) + 1) * .5)"}
     )
    )
    Scale RGB=&Multiply_VectorByFloat(
     Vector=<1 1 1>
     Float=&FloatFromCurve(FloatCurve=DataInterface() {propertyValues {Curve {
         defaultValue=1
         keys [
          <at 0.0704 0>
          {arriveTangent=2.6056, interpMode=RCIM_Cubic, leaveTangent=2.6056, time=0.25, value=0.6}
          {arriveTangent=5.8667, interpMode=RCIM_Cubic, leaveTangent=5.8667, time=0.7797, value=1.8481}
          <at 1 5 RCIM_Cubic>
         ]
        }}})
    )
    ScaleA=true
    ScaleRGB=true
   )
   &Color(
    Color=&MakeCustomLinearColorFromBool(
     Bool=$User.bIsPainted
     True Color=$Particles.Color
     False Color=&Multiply_LinearColorByFloat(
      Linear Color=&ColorFromCurve(ColorCurve=DataInterface() {propertyValues {
         AlphaCurve {keys [{arriveTangent=-1.04, interpMode=RCIM_Cubic, time=0.1656, value=0.7}, {interpMode=RCIM_Cubic, leaveTangent=1.0748, time=1, value=0}]}
         BlueCurve {keys [
           <at 0.0704 0.2083>
           {arriveTangent=2.526, interpMode=RCIM_Cubic, leaveTangent=2.526, time=0.1736, value=0.65}
           <at 0.7797 2>
           <at 1 5 RCIM_Cubic>
          ]}
         GreenCurve {keys [
           <at 0.0704 0>
           {arriveTangent=2.6056, interpMode=RCIM_Cubic, leaveTangent=2.6056, time=0.1736, value=0}
           <at 0.7797 1.8481>
           <at 1 4.6625 RCIM_Cubic>
          ]}
         RedCurve {keys [
           <at 0.0704 0.0868>
           {arriveTangent=0.1596, interpMode=RCIM_Cubic, leaveTangent=0.1596, time=0.1736, value=0.0542}
           <at 0.7797 0.2>
           <at 1 1 RCIM_Cubic>
          ]}
        }})
      Float=&Lerp_Float(
       A=&FloatFromCurve(FloatCurve=DataInterface() {propertyValues {Curve {keys [<at 0 5>,<at 0.6 0>]}}})
       B=&FloatFromCurve(FloatCurve=DataInterface() {propertyValues {Curve {keys [<at 0 2.5>,<at 0.6 0>]}}})
       Alpha=HlslExpression() {hlslExpression="lerp(0, .8, (sin(Engine.Time * 3.14) + 1) * .5)"}
      )
     )
    )
    Scale Color=&Multiply_VectorByFloat(
     Vector=<1 1 1>
     Float=&MakeCustomFloatFromBool(Bool=$User.bIsFrontendPreview, True Float=1, False Float=0.5)
    )
   )
  ]
  rendererClasses [&NiagaraRibbonRendererProperties]
  renderers [{rendererClass=&NiagaraRibbonRendererProperties}]
 }
 NE_VC_Boost_Sphere {
  emitterUpdateScript [
   &EmitterState(
    Inactive Response=::NewEnumerator1
    Life Cycle Mode=::NewEnumerator1
    Loop Behavior=::NewEnumerator1
    Loop Duration Mode=::NewEnumerator1
    UseLoopDelay=false
   )
   &SpawnBurst_Instantaneous(Use Spawn Probability=false, UseLoopCountLimit=false)
   &VC_FilteredSocketCount() {enabled=false}
   &SpawnRate(Use Spawn Probability=false) {enabled=false}
  ]
  particleSpawnScript [
   &VC_PaintedVariantsOverride(
    Black=<rgba 0.0085 0.0085 0.0085 0.5>
    BurntSienna=<rgba 0.12 0.0371 0.0148 1>
    Cobalt=<rgba 0.078 0.1798 1.3 1>
    Crimson=<rgba 1.5 0.03 0.0545 1>
    DefaultColor=&MakeCustomLinearColorFromBool(Bool=$User.bIsFrontendPreview, True Color=<rgba 0.05 0 1 1>, False Color=<rgba 0.08 0 0.8 1>)
    ForestGreen=<rgba 0.1625 0.5 0.05 1>
    Grey=<rgba 0.07 0.07 0.07 1>
    Lime=<rgba 0.2712 0.35 0.035 1>
    Orange=<rgba 1 0.1667 0 1>
    Pink=<rgba 0.9 0.27 0.9 1>
    Purple=<rgba 0.1204 0.056 0.7 1>
    R1 Black:LinearColor=<rgba 0.013 0.013 0.013 1.2>
    R1 BurntSienna:LinearColor=<rgba 0.04 0.0124 0.0049 1.35>
    R1 Cobalt:LinearColor=<rgba 0 0.0129 0.7 1.25>
    R1 Crimson:LinearColor=<rgba 0.26 0 0.007 1.25>
    R1 DefaultColor=<rgba 0.165 0.3987 1.1 0.3>
    R1 ForestGreen:LinearColor=<rgba 0 0.27 0.0136 1.25>
    R1 Grey:LinearColor=<rgba 0.12 0.12 0.12 1.1>
    R1 Lime:LinearColor=<rgba 0.175 0.35 0 1.15>
    R1 Orange:LinearColor=<rgba 0.38 0.1325 0 1.25>
    R1 Pink:LinearColor=<rgba 0.3 0.0359 0.204 1.25>
    R1 Purple:LinearColor=<rgba 0.1642 0 0.4 1>
    R1 Saffron:LinearColor=<rgba 0.5 0.3833 0 1>
    R1 SkyBlue:LinearColor=<rgba 0 0.1996 0.3 1>
    R1 TitaniumWhite:LinearColor=<rgba 0.45 0.45 0.45 1>
    Saffron=<rgba 0.5 0.345 0.035 1>
    SkyBlue=<rgba 0.1125 0.4312 0.75 1>
    TitaniumWhite=<rgba 0.7 0.7 0.7 1>
    Use R1 Painted Overrides=true
   )
   &InitializeParticle(
    Color=$Particles.Color
    Position Offset=&AddVectorToPosition(
     Position=&AddVectorToPosition(Position=<-90 0 0>, Vector=$User.SingleBoostPos)
     Vector=&Multiply_VectorByFloat(Vector=$User.DTCarOffset, Float=1)
    )
    Position Offset Coordinate Space=::Local
    Sprite Rotation Mode=::NewEnumerator2
    Sprite Size Mode=::NewEnumerator3
    Uniform Sprite Size=&VC_FloatBySpawnGroup(
     SpawnGroup or ExecIndex:NiagaraInt32=1
     CustomInt:NiagaraInt32=0
     MaximumSpawnGroups:NiagaraInt32=2
     Float SG 0:NiagaraFloat=115
     Float SG 1:NiagaraFloat=115
     Float SG 2:NiagaraFloat=0
     Float SG 3:NiagaraFloat=0
     Float SG 4:NiagaraFloat=0
     Float SG 5:NiagaraFloat=0
    )
    UsePositionOffset=true
    Write Material Random=true
   )
   &PartitionParticles(Number Of Partitions=$Emitter.PortCount) {enabled=false}
   &SkeletalMeshLocation(
    Bone / Socket Index=$Particles.PartitionParticles.Partition
    Bone / Socket Sampling Mode=::NewEnumerator5
    MaskSpawnGroup=false
    Mesh Sampling Type=::NewEnumerator6
    Skeletal Mesh=$System.SkeletalMesh
   ) {enabled=false}
   &AddVelocity(Velocity=<0 0 0>) {enabled=false}
  ]
  particleUpdateScript [
   &ParticleState(Kill Particles When Lifetime Has Elapsed=false, Loop Particles Lifetime=true)
   &Drag(
    Drag=&FloatFromCurve(FloatCurve=DataInterface() {propertyValues {Curve {keys [<at 0 1>,<at 0.2 0>]}}}, Scale Curve=10)
    Use Linear Drag=true
    Use Rotational Drag=true
   )
   &SolveForcesAndVelocity(Clamp Velocity=false, Limit Acceleration=false)
   &ScaleColor(
    Scale RGB=&Multiply_VectorByFloat(
     Vector=<1 1 1>
     Float=&MakeCustomFloatFromBool(
      Bool=$User.bIsFrontendPreview
      True Float=&MakeCustomFloatFromBool(Bool=$User.bIsPainted, True Float=5, False Float=15)
      False Float=&MakeCustomFloatFromBool(Bool=$User.bIsPainted, True Float=3.5, False Float=15)
     )
    )
    ScaleA=true
    ScaleRGB=true
   )
   /SetVariables(Particles.PositionShadow:NiagaraPosition=&AddVectorToPosition(
     Position=$Particles.Position
     Vector=&MakeVector(
      X=HlslExpression() {hlslExpression="sin((Engine.Time * 1.4)  * 6.2831)"}
      Y=HlslExpression() {hlslExpression="sin((Engine.Time * 1.4)  * 6.2831)"}
      Z=HlslExpression() {hlslExpression="sin((Engine.Time * 1.4)  * 6.2831)* 6"}
     )
    )) {isSetParameters=true}
  ]
  rendererClasses [&NiagaraSpriteRendererProperties]
  renderers [
   {rendererClass=&NiagaraSpriteRendererProperties}
   {rendererClass=&NiagaraSpriteRendererProperties, rendererIndex=1}
  ]
 }
}
summary {
 systemName=NS_Boost_ElectroGalactic_Optimized_Backup
 emitters [
  {emitterName=NE_VC_Boost_Sphere, rendererClasses [&NiagaraSpriteRendererProperties]}
  {emitterName=NE_Lightning_Core, rendererClasses [&NiagaraSpriteRendererProperties]}
  {emitterName=NE_Lightning_crumbs, rendererClasses [&NiagaraSpriteRendererProperties]}
  {emitterName=NE_VC_Boost_MAIN_RIBBON001, rendererClasses [&NiagaraRibbonRendererProperties]}
  {emitterName=NE_VC_Boost_SEC_RIBBON, rendererClasses [&NiagaraRibbonRendererProperties]}
 ]
 userVariables [
  User.bIsFrontendPreview {type=NiagaraBool}
  User.bIsPainted {type=NiagaraBool}
  User.ColorVariantID {type=NiagaraInt32}
  User.DTCarOffset {type=Vector3f, description="", defaultValue {struct=Vector3f, value=<1 1 1>}}
  User.PaintedColor {type=LinearColor, description="", defaultValue {struct=LinearColor, value=<rgba 1 1 1 1>}}
  User.PaintedScalar {type=NiagaraFloat, defaultValue=0, description=""}
  User.PortCount {type=NiagaraInt32, defaultValue=1}
  User.Scale Curve_Index 0 Param 2 {type=NiagaraFloat, defaultValue=5, description=""}
  User.SingleBoostPos {type=Vector3f, description="", defaultValue {struct=Vector3f, value=<0 0 0>}}
 ]
}
